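; Generic parent gameplay ability that holsters the activating pawn's weapon while active and unholsters it on end.
; On activation it derives a holster ID from its display name, overrides the current weapon's animset with the default player one, and holsters.
; A weapon-equipped delegate moves that override from previous to new weapon.
; Ending clears the override and releases the holster by ID, with no delegate unbind in this graph.
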(event Ability|EventActivateAbility
  (bind _returnvalue (Ability|GetActivatingPawn))
  ;[1 1 1] Activate - Holster
  (Variables|Default|SetHolster%20Id (Utilities|String|StringToName (Utilities|GetDisplayName self)))
  (CharacterAnimation|SetWeaponPawnAnimsetOverride (Variables|/Script/FortniteGame.FortPawn|GetCurrentWeapon _returnvalue) "/Game/Animation/Game/MainPlayer/AnimAssets/FortPlayerDefault_AnimAsset.FortPlayerDefault_AnimAsset")
  (Weapon|SetHolsterWeapon _returnvalue 0 true false (Variables|Default|GetHolster%20Id))
  (FortPawn|BindEventtoOnWeaponEquippedDelegate (Ability|GetActivatingPawn) (delegate Custom|WeaponEquipped)))

(event Custom|WeaponEquipped
  (params (NewWeapon :type Fort%20Weapon%20Object%20Reference) (PrevWeapon :type Fort%20Weapon%20Object%20Reference))
  ;[1 1 1] Activate - Holster
  (CharacterAnimation|SetWeaponPawnAnimsetOverride NewWeapon "/Game/Animation/Game/MainPlayer/AnimAssets/FortPlayerDefault_AnimAsset.FortPlayerDefault_AnimAsset")
  (CharacterAnimation|SetWeaponPawnAnimsetOverride PrevWeapon 0))

(event Ability|EventOnEndAbility
  (params (bWasCancelled :type Boolean))
  ;[1 1 1] On End Ability - Unholster
  (CharacterAnimation|SetWeaponPawnAnimsetOverride (Variables|/Script/FortniteGame.FortPawn|GetCurrentWeapon (Ability|GetActivatingPawn)) 0)
  (Weapon|SetHolsterWeapon (Ability|GetActivatingPawn) 0 false true (Variables|Default|GetHolster%20Id)))

(event Ability|EventFailedToActivatePassiveAbility
  (params (ActorInfo :type Gameplay%20Ability%20Actor%20Info%20Structure))
  (|CallParentFunction))
